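; AvacadoEater projectile Blueprint that carries a team and spawns a deployed gameplay actor through its parent's SpawnBGA_.
; SetTeam stores the team and flags it as set; BeginPlay derives it from the instigating FortPawn when unset.
; SpawnBGA_ casts the parent's result to the AvacadoEater deployed parent and checks the team flag, but that branch is empty, so it only returns the spawned actor.
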(event Custom|SetTeam
  (params (Team :type Byte))
  (Variables|Default|SetTeam Team)
  (Variables|Default|SetTeamHasBeenSet true))

(event EventBeginPlay
  (if (not (Variables|Default|GetTeamHasBeenSet))
    (bind _asfort_pawn (Utilities|Casting|CastToFortPawn (Game|GetInstigator))
      (:then
        (CallFunction|SetTeam (Team|GetActorTeam _asfort_pawn)))
      (:CastFailed
        _))
    (|CallParentFunction)))

(event Collision|EventActorBeginOverlap
  (params (OtherActor :type Actor%20Object%20Reference)))

(event EventTick
  (params (DeltaSeconds :type Float%20%28single-precision%29)))

(fn SpawnBGA_
  (params (Normal :type Vector :default "0, 0, 0") (HitActor :type Actor%20Object%20Reference) (HitComp :type Primitive%20Component%20Object%20Reference) (Pos :type Vector :default "0, 0, 0") (FromWater :type Boolean) (Hit :type Hit%20Result%20Structure))
  (bind _returnvalue (|CallParentFunction Normal HitActor HitComp Pos FromWater))
  (Utilities|FlowControl|Sequence
    (:then_0
      (Utilities|Casting|CastToBGA_Athena_AvacadoEater_Deployed_Parent _returnvalue)
      (if (Variables|Default|GetTeamHasBeenSet)
        _))
    (:then_1
      (return _returnvalue))))

(fn ConstructionScript
  (|CallParentFunction))
